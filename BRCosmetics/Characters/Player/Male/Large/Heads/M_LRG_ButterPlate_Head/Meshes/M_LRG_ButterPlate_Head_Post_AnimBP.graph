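# Post-process AnimBlueprint for the male large ButterPlate head that adjusts the upper-body pose for the frontend backpack pose.
# The update event branches on bIsInFrontEnd and copies the Frontend_Backpack_Pose curve value into a variable.
# That variable is the alpha for a chain of additive Transform Bone nodes that rotate the five spine bones, head and neck_01 on top of the input pose.
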AnimBlueprint /BRCosmetics/Characters/Player/Male/Large/Heads/M_LRG_ButterPlate_Head/Meshes/M_LRG_ButterPlate_Head_Post_AnimBP
target_skeleton=/Game/Characters/Player/Male/Large/Heads/M_LRG_ButterPlate_Head/Meshes/CustomSkeleton/M_LRG_ButterPlate_Head_Skeleton
graphs [
 AnimGraph {nodes [
   n1 = Get Frontend_Backpack_Pose() -> (Frontend_Backpack_Pose:float)
   n2 = "Transform (Modify) Bone - Bone: spine_05"(ComponentPose=Local To Component(LocalPose=Input Pose()), Alpha=@n1, Translation="0,0,0", Rotation="0,0,5", Scale="1,1,1") {rotationMode=BMM_Additive, alphaBoolBlend {blendOption=Linear}, boneToModify {boneName=spine_05}, rotation {roll=5}}
   n3 = "Transform (Modify) Bone - Bone: spine_04"(ComponentPose=@n2, Alpha=@n1, Translation="0,0,0", Rotation="0,0,5", Scale="1,1,1") {rotationMode=BMM_Additive, alphaBoolBlend {blendOption=Linear}, boneToModify {boneName=spine_04}, rotation {roll=5}}
   n4 = "Transform (Modify) Bone - Bone: spine_03"(ComponentPose=@n3, Alpha=@n1, Translation="0,0,0", Rotation="3,0,10", Scale="1,1,1") {rotationMode=BMM_Additive, alphaBoolBlend {blendOption=Linear}, boneToModify {boneName=spine_03}, rotation {pitch=3, roll=10}}
   n5 = "Transform (Modify) Bone - Bone: spine_02"(ComponentPose=@n4, Alpha=@n1, Translation="0,0,0", Rotation="3,0,10", Scale="1,1,1") {rotationMode=BMM_Additive, alphaBoolBlend {blendOption=Linear}, boneToModify {boneName=spine_02}, rotation {pitch=3, roll=10}}
   n6 = "Transform (Modify) Bone - Bone: spine_01"(ComponentPose=@n5, Alpha=@n1, Translation="0,0,0", Rotation="6,0,10", Scale="1,1,1") {rotationMode=BMM_Additive, alphaBoolBlend {blendOption=Linear}, boneToModify {boneName=spine_01}, rotation {pitch=6, roll=10}}
   n7 = "Transform (Modify) Bone - Bone: head"(ComponentPose=@n6, Alpha=@n1, Translation="0,0,0", Rotation="0,0,-30", Scale="1,1,1") {rotationMode=BMM_Additive, alphaBoolBlend {blendOption=Linear}, boneToModify {boneName=head}, rotation {roll=-30}}
   n8 = "Transform (Modify) Bone - Bone: neck_01"(ComponentPose=@n7, Alpha=@n1, Translation="0,0,0", Rotation="-5.113,15,-10", Scale="1,1,1") {rotationMode=BMM_Additive, alphaBoolBlend {blendOption=Linear}, boneToModify {boneName=neck_01}, rotation=<rot -5.113 15 -10>}
   Output Pose(Result=Component To Local(ComponentPose=@n8))
  ]}
 EventGraph {nodes [
   Event BlueprintUpdateAnimation()
   > Branch(Condition=Get bIsInFrontEnd())
   > Set Frontend_Backpack_Pose(Frontend_Backpack_Pose:float=GetCurveValue(CurveName=Frontend_Backpack_Pose)) -> (Output_Get:float)
   TryGetPawnOwner()
  ]}
]
variables [Frontend_Backpack_Pose {type=number}]
